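; Thrown Gas Flopper projectile that becomes a timed, damaging gas cloud on stopping.
; On stop it hides the fish; authority aligns the first DoExplosion to whole game-time seconds, repeats it on a timer, then destroys the actor after the gas duration.
; Remote cosmetics spawn impact FX and sound, then fade smoke material, audio and post-process through one timeline played forward and reversed.
; Gas duration comes from AthenaGameData; tick rate is hardcoded when its row is found.
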(event Custom|Stop_Rotation
  ;[1 1 1] Stop Rotating Movement
  (Components|Movement|SetUpdatedComponent (Variables|Default|GetRotatingMovement) 0))

(event Custom|DamageTick
  (Delivery|DoExplosion)
  (Development|PrintString "Exploded" true true "(R=0.000000,G=0.660000,B=1.000000,A=1.000000)" 0.5))

(event Custom|SyncedTimer
  ;[1 1 1] Sync the first explosion with 1 second game time ticks, then start the rest of the timer
  (CallFunction|DamageTick)
  (Utilities|Time|SetTimerbyFunctionName 0 "DamageTick" (Variables|GasGrenadeTuning|GetDoT_TickRate) true))

(event Custom|Update%20PostProcess%20Scale)

(event Custom|Update%20Tick%20Rate
  (DataTable|EvaluateCurveTableRow "/Game/Athena/Balance/DataTables/AthenaGameData.AthenaGameData" "Default.GasGrenade.DamagePeriod"
    (:RowFound
      (Variables|GasGrenadeTuning|SetDoT_TickRate 0.5))))

(event Custom|Update%20Gas%20Duration
  (bind _outxy (DataTable|EvaluateCurveTableRow "/Game/Athena/Balance/DataTables/AthenaGameData.AthenaGameData" "Default.GasGrenade.GasDuration")
    (:RowFound
      (Variables|GasGrenadeTuning|SetGasDuration _outxy))))

(event OnEnterWaterEvent(WaterInteractionComponent)
  (params (WaterBody :type Fort%20Water%20Body%20Actor%20Object%20Reference) (WaterInteractionComponent :type Fort%20Water%20Interaction%20Component%20Object%20Reference) (bIsFirstBody :type Boolean))
  (Variables|Default|SetHitWater true))

(event EventOnShot
  (params (Hit :type Hit%20Result%20Structure%20%28by%20ref%29))
  (Development|PrintString "Thrown" true true "(R=0.000000,G=0.660000,B=1.000000,A=1.000000)" 5e+01))

(event EventOnStop
  (params (Hit :type Hit%20Result%20Structure%20%28by%20ref%29))
  (bind _returnvalue (Utilities|Time|GetGameTimeinSeconds))
  (bind _returnvalue_2 (Math|Float|Truncate _returnvalue))
  (bind _returnvalue_3 (Math|Conversions|ToFloat(Integer) _returnvalue_2))
  (bind _returnvalue_4 (Utilities|Time|GetGameTimeinSeconds))
  (bind _returnvalue_5 (- _returnvalue_4 _returnvalue_3))
  (bind _returnvalue_6 (- 1.0 _returnvalue_5))
  (bind _returnvalue_7 (Transformation|GetActorLocation))
  (bind (_bblockinghit _binitialoverlap _time _distance _location _impactpoint _normal _impactnormal) (Collision|BreakHitResult Hit))
  (bind _dot_tickrate (Variables|GasGrenadeTuning|GetDoT_TickRate))
  ;[0.7116 0.838 1] OnStop - Stop the rotation and explode.
  (CallFunction|StopRotation)
  (Development|SetHiddeninGame (Variables|Default|GetGas_Fish) true true)
  (Collision|SetCollisionEnabled (Variables|Default|GetOverlapCapsule) "NoCollision")
  (CallFunction|UpdatePostProcessScale)
  (CallFunction|UpdateGasDuration)
  (Networking|SwitchHasAuthority
    (:Authority
      (Utilities|FlowControl|Sequence
        (:then_0
          (CallFunction|UpdateTickRate)
          ;[1 1 1] Explode on a timer to deal damage.
          (if (>= _returnvalue_6 _dot_tickrate)
            (Variables|Default|SetSyncedDelay (- _returnvalue_6 _dot_tickrate))
            (else
              (Variables|Default|SetSyncedDelay _returnvalue_6)))
          (Utilities|Time|SetTimerbyFunctionName 0 "SyncedTimer" (Variables|Default|GetSyncedDelay))
          (Development|PrintString (Utilities|String|BuildString(Float) (Utilities|String|BuildString(Float) (Utilities|String|BuildString(Float) "" "game time " _returnvalue_4) " minus truncated game time " _returnvalue_3) " equals this " _returnvalue_5) false false "(R=0.000000,G=0.660000,B=1.000000,A=1.000000)" 2e+01))
        (:then_1
          ;[1 1 1] Destroy when done.
          (Utilities|FlowControl|Delay (Variables|GasGrenadeTuning|GetGasDuration))
          (Utilities|Time|ClearTimerbyFunctionName 0 "DamageTick")
          (Actor|DestroyActor))))
    (:Remote
      ;[1 1 1] OnExploded - Emitter, Sound, Camera Shake
      (Networking|CanExecuteCosmeticEvents
        (:True
          (bind _returnvalue_8 (Effects|SpawnEmitteratLocation "/Game/Athena/Items/Consumables/GasGrenade/P_GasGrenade_Impact_Explosion.P_GasGrenade_Impact_Explosion" _returnvalue_7 (Math|Conversions|RotationFromXVector (select (Variables|Default|GetHitWater) "0, 0,1.000000" _impactnormal)) (Transformation|GetWorldScale (Variables|Default|GetPostProcessParentShape))))
          (Rendering|SetTranslucentSortPriority _returnvalue_8 -1)
          (Audio|PlaySoundatLocation (Variables|Sounds|GetCue_CloseSound) _returnvalue_7) ; Play Close Explosion
          (Audio|Components|Audio|FadeOut (Variables|B_Prj_Grenade|GetGrenadeFuse_AudioComponent) 0.2 0.0 "Linear" :!class /Script/Engine.AudioComponent) ; Fade Out Fuse Sound
          (Actor|SetLifeSpan (+ (Variables|GasGrenadeTuning|GetGasDuration) (Variables|Default|GetGasFadeInTime)))
          (Variables|/Script/Engine.PostProcessComponent|SetBlendWeight 0.0 (Variables|Default|GetPostProcessComponent))
          (Utilities|Time|SetTimerbyFunctionName 0 "FadeOut" (- (Variables|GasGrenadeTuning|GetGasDuration) (Variables|Default|GetGasFadeOutTime)))
          (Components|Activation|Activate (Variables|Default|GetP_GasGrenade_InnerSmoke) false)
          (Variables|/Script/Engine.PrimitiveComponent|SetbGenerateOverlapEvents true (Variables|Default|GetPostProcessParentShape))
          (Components|Activation|Activate (Variables|Default|GetAudioReleaseSmoke) false)
          (Utilities|FlowControl|Delay (Variables|Default|GetGasDelayBeforeFadeIn))
          (CallFunction|FadeIn))))))

(event Custom|FadeIn
  ;[1 1 1] Timeline
  (Components|Timeline|SetPlayRate (Variables|B_Prj_Athena_GasFlopper|GetFadeSmokeTL) (Utilities|Reciprocal (Variables|Default|GetGasFadeInTime)))
  (Rendering|SetVisibility (Variables|Default|GetSmokeVolumeMesh) true false)
  (bind (_direction _float) (Timeline|FadeSmokeTL PlayFromStart)
    (:Update
      (Rendering|Material|SetScalarParameterValue (Variables|Default|GetSmokeVolumeMeshMID) "Fade" _float :!class /Script/Engine.MaterialInstanceDynamic)
      (Audio|Components|Audio|SetVolumeMultiplier (Variables|Default|GetAudioReleaseSmoke) _float :!class /Script/Engine.AudioComponent)
      (Variables|/Script/Engine.PostProcessComponent|SetBlendWeight (* _float 0.7) (Variables|Default|GetPostProcessComponent)))
    (:Finished
      (Development|PrintString "this faded"))))

(event Custom|FadeOut
  ;[1 1 1] Timeline
  (Components|Timeline|SetPlayRate (Variables|B_Prj_Athena_GasFlopper|GetFadeSmokeTL) (Utilities|Reciprocal (Variables|Default|GetGasFadeOutTime)))
  (Components|Activation|Deactivate (Variables|Default|GetP_GasGrenade_InnerSmoke))
  (Timeline|FadeSmokeTL ReverseFromEnd))

(fn ConstructionScript
  (Networking|CanExecuteCosmeticEvents
    (:True
      (Components|Movement|SetUpdatedComponent (Variables|Default|GetRotatingMovement) (Variables|Default|GetGas_Fish))
      (Components|Capsule|SetCapsuleSize (Variables|Default|GetOverlapCapsule) (Variables|/Script/Engine.CapsuleComponent|GetCapsuleRadius (Variables|Collision|GetCapsuleComponent)) (Variables|/Script/Engine.CapsuleComponent|GetCapsuleHalfHeight (Variables|Collision|GetCapsuleComponent)) true)
      (bind _returnvalue (Rendering|Material|CreateDynamicMaterialInstance (Variables|Default|GetSmokeVolumeMesh) 0 "/Game/Athena/Items/Consumables/Flopper/GasFlopper/VFX/MI_GasFlopper_GasGrenade_OuterShell.MI_GasFlopper_GasGrenade_OuterShell" "None" :!class /Script/Engine.PrimitiveComponent))
      (bind _output_get (Variables|Default|SetSmokeVolumeMeshMID _returnvalue))
      (Rendering|Material|SetScalarParameterValue _output_get "Radius" (Variables|Default|GetGasGrenadeRadius) :!class /Script/Engine.MaterialInstanceDynamic)
      (Variables|/Script/Engine.PostProcessComponent|SetBlendWeight 0.0 (Variables|Default|GetPostProcessComponent)))))
